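# Animation Blueprint for the IvyStub mic stand pickaxe that animates floating stones, handle and mouthpiece.
# The event graph derives sine/cosine bobs and steady spins from game time, applied by additive bone transforms to the input pose.
# Tightly limited spring AnimDynamics follow, and the result is cached.
# While bIsPlayingMeleeAnim is true, the output blends to that cached pose plus a second, wider-limit spring pass on most stones.
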AnimBlueprint /SparksCosmetics/Instrument/MicStand/IvyStub/Meshes/Pickaxe/IvyStub_MicStand_Pickaxe_AnimBP
target_skeleton=/SparksCosmetics/Instrument/MicStand/IvyStub/Meshes/Backpack/IvyStub_MicStand_Backpack_Skeleton
graphs [
 AnimGraph {nodes [
   Copy Pose From Mesh() {bCopyCurves=true, bUseAttachedParent=true}
   "Transform (Modify) Bone - Bone: stone_04_large"(Alpha=1, Translation="0,0,0", Rotation="0,0,0", Scale="1,1,1") {
    rotationMode=BMM_Additive
    translationMode=BMM_Additive
    translationSpace=BCS_BoneSpace
    alphaBoolBlend {blendOption=Linear}
    boneToModify {boneName=stone_04_large}
   }
   n1 = Anim Dynamics - Bone: stone_02_large(ComponentPose=Local To Component(LocalPose=Use cached pose 'Cached Pose 1'()), Alpha=0.8, ExternalForce="0,0,0") {
    angularDampingOverride=0.9
    angularSpringConstant=100
    bAngularSpring=true
    bGravityOverrideInSimSpace=true
    bLinearSpring=true
    bOverrideAngularDamping=true
    bOverrideLinearDamping=true
    gravityScale=0
    linearDampingOverride=0.8
    linearSpringConstant=30
    alphaBoolBlend {blendOption=Linear}
    boundBone {boneName=stone_02_large}
    componentAppliedLinearAccClamp=<10000 10000 10000>
    componentLinearAccScale=<0.15 0.15 0.15>
    componentLinearVelScale=<0.8 0.8 0.6>
    physicsBodyDefinitions [{
      sphereCollisionRadius=0
      boundBone {boneName=stone_02_large}
      boxExtents=<50 50 50>
      constraintSetup {angularLimitsMax=<0 0 0>, angularLimitsMin=<0 0 0>, angularTarget=<0 0 0>, linearAxesMax=<2 5 8>, linearAxesMin=<-2 -1 -8>}
      localJointOffset=<0 -3 0>
     }]
    relativeSpaceBone {boneName=handle}
   }
   n2 = Anim Dynamics - Bone: stone_03_large(ComponentPose=@n1, Alpha=0.8, ExternalForce="0,0,0") {
    angularDampingOverride=0.9
    angularSpringConstant=100
    bAngularSpring=true
    bGravityOverrideInSimSpace=true
    bLinearSpring=true
    bOverrideAngularDamping=true
    bOverrideLinearDamping=true
    gravityScale=0
    linearDampingOverride=0.8
    linearSpringConstant=30
    alphaBoolBlend {blendOption=Linear}
    boundBone {boneName=stone_03_large}
    componentAppliedLinearAccClamp=<10000 10000 10000>
    componentLinearAccScale=<0.15 0.15 0.15>
    componentLinearVelScale=<0.8 0.8 0.6>
    physicsBodyDefinitions [{
      sphereCollisionRadius=0
      boundBone {boneName=stone_03_large}
      boxExtents=<50 50 50>
      constraintSetup {angularLimitsMax=<0 0 0>, angularLimitsMin=<0 0 0>, angularTarget=<0 0 0>, linearAxesMax=<2 5 8>, linearAxesMin=<-2 -1 -8>}
      localJointOffset=<0 -3 0>
     }]
    relativeSpaceBone {boneName=handle}
   }
   n3 = Anim Dynamics - Bone: stone_05_large(ComponentPose=@n2, Alpha=0.8, ExternalForce="0,0,0") {
    angularDampingOverride=0.9
    angularSpringConstant=100
    bAngularSpring=true
    bGravityOverrideInSimSpace=true
    bLinearSpring=true
    bOverrideAngularDamping=true
    bOverrideLinearDamping=true
    gravityScale=0
    linearDampingOverride=0.8
    linearSpringConstant=30
    alphaBoolBlend {blendOption=Linear}
    boundBone {boneName=stone_05_large}
    componentAppliedLinearAccClamp=<10000 10000 10000>
    componentLinearAccScale=<0.15 0.15 0.15>
    componentLinearVelScale=<0.8 0.8 0.6>
    physicsBodyDefinitions [{
      sphereCollisionRadius=0
      boundBone {boneName=stone_05_large}
      boxExtents=<50 50 50>
      constraintSetup {angularLimitsMax=<0 0 0>, angularLimitsMin=<0 0 0>, angularTarget=<0 0 0>, linearAxesMax=<2 5 8>, linearAxesMin=<-2 0 -8>}
      localJointOffset=<0 -3 0>
     }]
    relativeSpaceBone {boneName=handle}
   }
   n4 = Anim Dynamics - Bone: stone_01_small(ComponentPose=@n3, Alpha=0.8, ExternalForce="0,0,0") {
    angularDampingOverride=0.9
    angularSpringConstant=100
    bAngularSpring=true
    bGravityOverrideInSimSpace=true
    bLinearSpring=true
    bOverrideAngularDamping=true
    bOverrideLinearDamping=true
    gravityScale=0
    linearDampingOverride=0.8
    linearSpringConstant=30
    alphaBoolBlend {blendOption=Linear}
    boundBone {boneName=stone_01_small}
    componentAppliedLinearAccClamp=<10000 10000 10000>
    componentLinearAccScale=<0.15 0.15 0.15>
    componentLinearVelScale=<0.8 0.8 0.6>
    physicsBodyDefinitions [{
      sphereCollisionRadius=0
      boundBone {boneName=stone_01_small}
      boxExtents=<50 50 50>
      constraintSetup {angularLimitsMax=<0 0 0>, angularLimitsMin=<0 0 0>, angularTarget=<0 0 0>, linearAxesMax=<2 5 8>, linearAxesMin=<-2 0 -8>}
      localJointOffset=<0 -3 0>
     }]
    relativeSpaceBone {boneName=handle}
   }
   n5 = Anim Dynamics - Bone: stone_02_small(ComponentPose=@n4, Alpha=0.8, ExternalForce="0,0,0") {
    angularDampingOverride=0.9
    angularSpringConstant=100
    bAngularSpring=true
    bGravityOverrideInSimSpace=true
    bLinearSpring=true
    bOverrideAngularDamping=true
    bOverrideLinearDamping=true
    gravityScale=0
    linearDampingOverride=0.8
    linearSpringConstant=30
    alphaBoolBlend {blendOption=Linear}
    boundBone {boneName=stone_02_small}
    componentAppliedLinearAccClamp=<10000 10000 10000>
    componentLinearAccScale=<0.15 0.15 0.15>
    componentLinearVelScale=<0.8 0.8 0.6>
    physicsBodyDefinitions [{
      sphereCollisionRadius=0
      boundBone {boneName=stone_02_small}
      boxExtents=<50 50 50>
      constraintSetup {angularLimitsMax=<0 0 0>, angularLimitsMin=<0 0 0>, angularTarget=<0 0 0>, linearAxesMax=<2 5 8>, linearAxesMin=<-2 1 -8>}
      localJointOffset=<0 -3 0>
     }]
    relativeSpaceBone {boneName=handle}
   }
   n6 = Anim Dynamics - Bone: stone_03_small(ComponentPose=@n5, Alpha=0.8, ExternalForce="0,0,0") {
    angularDampingOverride=0.9
    angularSpringConstant=100
    bAngularSpring=true
    bGravityOverrideInSimSpace=true
    bLinearSpring=true
    bOverrideAngularDamping=true
    bOverrideLinearDamping=true
    gravityScale=0
    linearDampingOverride=0.8
    linearSpringConstant=30
    alphaBoolBlend {blendOption=Linear}
    boundBone {boneName=stone_03_small}
    componentAppliedLinearAccClamp=<10000 10000 10000>
    componentLinearAccScale=<0.15 0.15 0.15>
    componentLinearVelScale=<0.8 0.8 0.6>
    physicsBodyDefinitions [{
      sphereCollisionRadius=0
      boundBone {boneName=stone_03_small}
      boxExtents=<50 50 50>
      constraintSetup {angularLimitsMax=<0 0 0>, angularLimitsMin=<0 0 0>, angularTarget=<0 0 0>, linearAxesMax=<2 5 8>, linearAxesMin=<-2 1 -8>}
      localJointOffset=<0 -3 0>
     }]
    relativeSpaceBone {boneName=handle}
   }
   n7 = Anim Dynamics - Bone: stone_04_small(ComponentPose=@n6, Alpha=0.8, ExternalForce="0,0,0") {
    angularDampingOverride=0.9
    angularSpringConstant=100
    bAngularSpring=true
    bGravityOverrideInSimSpace=true
    bLinearSpring=true
    bOverrideAngularDamping=true
    bOverrideLinearDamping=true
    gravityScale=0
    linearDampingOverride=0.8
    linearSpringConstant=30
    alphaBoolBlend {blendOption=Linear}
    boundBone {boneName=stone_04_small}
    componentAppliedLinearAccClamp=<10000 10000 10000>
    componentLinearAccScale=<0.15 0.15 0.15>
    componentLinearVelScale=<0.8 0.8 0.6>
    physicsBodyDefinitions [{
      sphereCollisionRadius=0
      boundBone {boneName=stone_04_small}
      boxExtents=<50 50 50>
      constraintSetup {angularLimitsMax=<0 0 0>, angularLimitsMin=<0 0 0>, angularTarget=<0 0 0>, linearAxesMax=<2 5 8>, linearAxesMin=<-2 1 -8>}
      localJointOffset=<0 -3 0>
     }]
    relativeSpaceBone {boneName=handle}
   }
   n8 = Anim Dynamics - Bone: stone_05_small(ComponentPose=@n7, Alpha=0.8, ExternalForce="0,0,0") {
    angularDampingOverride=0.9
    angularSpringConstant=100
    bAngularSpring=true
    bGravityOverrideInSimSpace=true
    bLinearSpring=true
    bOverrideAngularDamping=true
    bOverrideLinearDamping=true
    gravityScale=0
    linearDampingOverride=0.8
    linearSpringConstant=30
    alphaBoolBlend {blendOption=Linear}
    boundBone {boneName=stone_05_small}
    componentAppliedLinearAccClamp=<10000 10000 10000>
    componentLinearAccScale=<0.15 0.15 0.15>
    componentLinearVelScale=<0.8 0.8 0.6>
    physicsBodyDefinitions [{
      sphereCollisionRadius=0
      boundBone {boneName=stone_05_small}
      boxExtents=<50 50 50>
      constraintSetup {angularLimitsMax=<0 0 0>, angularLimitsMin=<0 0 0>, angularTarget=<0 0 0>, linearAxesMax=<2 5 8>, linearAxesMin=<-2 1 -8>}
      localJointOffset=<0 -3 0>
     }]
    relativeSpaceBone {boneName=handle}
   }
   n9 = Anim Dynamics - Bone: stone_06_small(ComponentPose=@n8, Alpha=0.8, ExternalForce="0,0,0") {
    angularDampingOverride=0.9
    angularSpringConstant=100
    bAngularSpring=true
    bGravityOverrideInSimSpace=true
    bLinearSpring=true
    bOverrideAngularDamping=true
    bOverrideLinearDamping=true
    gravityScale=0
    linearDampingOverride=0.8
    linearSpringConstant=30
    alphaBoolBlend {blendOption=Linear}
    boundBone {boneName=stone_06_small}
    componentAppliedLinearAccClamp=<10000 10000 10000>
    componentLinearAccScale=<0.15 0.15 0.15>
    componentLinearVelScale=<0.8 0.8 0.6>
    physicsBodyDefinitions [{
      sphereCollisionRadius=0
      boundBone {boneName=stone_06_small}
      boxExtents=<50 50 50>
      constraintSetup {angularLimitsMax=<0 0 0>, angularLimitsMin=<0 0 0>, angularTarget=<0 0 0>, linearAxesMax=<2 5 8>, linearAxesMin=<-2 1 -8>}
      localJointOffset=<0 -3 0>
     }]
    relativeSpaceBone {boneName=handle}
   }
   n10 = Anim Dynamics - Bone: stone_07_small(ComponentPose=@n9, Alpha=0.8, ExternalForce="0,0,0") {
    angularDampingOverride=0.9
    angularSpringConstant=100
    bAngularSpring=true
    bGravityOverrideInSimSpace=true
    bLinearSpring=true
    bOverrideAngularDamping=true
    bOverrideLinearDamping=true
    gravityScale=0
    linearDampingOverride=0.8
    linearSpringConstant=30
    alphaBoolBlend {blendOption=Linear}
    boundBone {boneName=stone_07_small}
    componentAppliedLinearAccClamp=<10000 10000 10000>
    componentLinearAccScale=<0.15 0.15 0.15>
    componentLinearVelScale=<0.8 0.8 0.6>
    physicsBodyDefinitions [{
      sphereCollisionRadius=0
      boundBone {boneName=stone_07_small}
      boxExtents=<50 50 50>
      constraintSetup {angularLimitsMax=<0 0 0>, angularLimitsMin=<0 0 0>, angularTarget=<0 0 0>, linearAxesMax=<2 5 8>, linearAxesMin=<-2 1 -8>}
      localJointOffset=<0 -3 0>
     }]
    relativeSpaceBone {boneName=handle}
   }
   n11 = Anim Dynamics - Bone: stone_08_small(ComponentPose=@n10, Alpha=0.8, ExternalForce="0,0,0") {
    angularDampingOverride=0.9
    angularSpringConstant=100
    bAngularSpring=true
    bGravityOverrideInSimSpace=true
    bLinearSpring=true
    bOverrideAngularDamping=true
    bOverrideLinearDamping=true
    gravityScale=0
    linearDampingOverride=0.8
    linearSpringConstant=30
    alphaBoolBlend {blendOption=Linear}
    boundBone {boneName=stone_08_small}
    componentAppliedLinearAccClamp=<10000 10000 10000>
    componentLinearAccScale=<0.15 0.15 0.15>
    componentLinearVelScale=<0.8 0.8 0.6>
    physicsBodyDefinitions [{
      sphereCollisionRadius=0
      boundBone {boneName=stone_08_small}
      boxExtents=<50 50 50>
      constraintSetup {angularLimitsMax=<0 0 0>, angularLimitsMin=<0 0 0>, angularTarget=<0 0 0>, linearAxesMax=<2 5 8>, linearAxesMin=<-2 1 -8>}
      localJointOffset=<0 -3 0>
     }]
    relativeSpaceBone {boneName=handle}
   }
   n12 = Anim Dynamics - Bone: stone_09_small(ComponentPose=@n11, Alpha=0.8, ExternalForce="0,0,0") {
    angularDampingOverride=0.9
    angularSpringConstant=100
    bAngularSpring=true
    bGravityOverrideInSimSpace=true
    bLinearSpring=true
    bOverrideAngularDamping=true
    bOverrideLinearDamping=true
    gravityScale=0
    linearDampingOverride=0.8
    linearSpringConstant=30
    alphaBoolBlend {blendOption=Linear}
    boundBone {boneName=stone_09_small}
    componentAppliedLinearAccClamp=<10000 10000 10000>
    componentLinearAccScale=<0.15 0.15 0.15>
    componentLinearVelScale=<0.8 0.8 0.6>
    physicsBodyDefinitions [{
      sphereCollisionRadius=0
      boundBone {boneName=stone_09_small}
      boxExtents=<50 50 50>
      constraintSetup {angularLimitsMax=<0 0 0>, angularLimitsMin=<0 0 0>, angularTarget=<0 0 0>, linearAxesMax=<2 5 8>, linearAxesMin=<-2 1 -8>}
      localJointOffset=<0 -3 0>
     }]
    relativeSpaceBone {boneName=handle}
   }
   Output Pose(Result=Blend Poses by bool(
     BlendPose_0=Component To Local(ComponentPose=@n12)
     BlendPose_1=Use cached pose 'Cached Pose 1'()
     BlendTime_0=1
     BlendTime_1=2
     bActiveValue=Get bIsPlayingMeleeAnim()
    ) {blendType=Linear})
   n13 = Get StoneFloatA() -> (StoneFloatA:Vector)
   n14 = Get Rotator1() -> (Rotator1:Rotator)
   n15 = "Transform (Modify) Bone - Bone: stone_01_small"(ComponentPose=Local To Component(LocalPose=Input Pose()), Alpha=1, Translation=@n13, Rotation=@n14, Scale="1,1,1") {
    rotationMode=BMM_Additive
    translationMode=BMM_Additive
    translationSpace=BCS_BoneSpace
    alphaBoolBlend {blendOption=Linear}
    boneToModify {boneName=stone_01_small}
   }
   n16 = Get StoneFloatB() -> (StoneFloatB:Vector)
   n17 = Get Rotator2() -> (Rotator2:Rotator)
   n18 = "Transform (Modify) Bone - Bone: stone_02_small"(ComponentPose=@n15, Alpha=1, Translation=@n16, Rotation=@n17, Scale="1,1,1") {
    rotationMode=BMM_Additive
    translationMode=BMM_Additive
    translationSpace=BCS_BoneSpace
    alphaBoolBlend {blendOption=Linear}
    boneToModify {boneName=stone_02_small}
   }
   n19 = Get StoneFloatC() -> (StoneFloatC:Vector)
   n20 = "Transform (Modify) Bone - Bone: stone_03_small"(ComponentPose=@n18, Alpha=1, Translation=@n19, Rotation=@n14, Scale="1,1,1") {
    rotationMode=BMM_Additive
    translationMode=BMM_Additive
    translationSpace=BCS_BoneSpace
    alphaBoolBlend {blendOption=Linear}
    boneToModify {boneName=stone_03_small}
   }
   n21 = Get Shard_Float_D() -> (Shard_Float_D:Vector)
   n22 = "Transform (Modify) Bone - Bone: stone_04_small"(ComponentPose=@n20, Alpha=1, Translation=@n21, Rotation=@n17, Scale="1,1,1") {
    rotationMode=BMM_Additive
    translationMode=BMM_Additive
    translationSpace=BCS_BoneSpace
    alphaBoolBlend {blendOption=Linear}
    boneToModify {boneName=stone_04_small}
   }
   n23 = "Transform (Modify) Bone - Bone: stone_05_small"(ComponentPose=@n22, Alpha=1, Translation=@n13, Rotation=@n14, Scale="1,1,1") {
    rotationMode=BMM_Additive
    translationMode=BMM_Additive
    translationSpace=BCS_BoneSpace
    alphaBoolBlend {blendOption=Linear}
    boneToModify {boneName=stone_05_small}
   }
   n24 = "Transform (Modify) Bone - Bone: stone_06_small"(ComponentPose=@n23, Alpha=1, Translation=@n16, Rotation=@n17, Scale="1,1,1") {
    rotationMode=BMM_Additive
    translationMode=BMM_Additive
    translationSpace=BCS_BoneSpace
    alphaBoolBlend {blendOption=Linear}
    boneToModify {boneName=stone_06_small}
   }
   n25 = "Transform (Modify) Bone - Bone: stone_07_small"(ComponentPose=@n24, Alpha=1, Translation=@n19, Rotation=@n14, Scale="1,1,1") {
    rotationMode=BMM_Additive
    translationMode=BMM_Additive
    translationSpace=BCS_BoneSpace
    alphaBoolBlend {blendOption=Linear}
    boneToModify {boneName=stone_07_small}
   }
   n26 = "Transform (Modify) Bone - Bone: stone_08_small"(ComponentPose=@n25, Alpha=1, Translation=@n21, Rotation=@n17, Scale="1,1,1") {
    rotationMode=BMM_Additive
    translationMode=BMM_Additive
    translationSpace=BCS_BoneSpace
    alphaBoolBlend {blendOption=Linear}
    boneToModify {boneName=stone_08_small}
   }
   n27 = "Transform (Modify) Bone - Bone: stone_09_small"(ComponentPose=@n26, Alpha=1, Translation=@n13, Rotation=@n14, Scale="1,1,1") {
    rotationMode=BMM_Additive
    translationMode=BMM_Additive
    translationSpace=BCS_BoneSpace
    alphaBoolBlend {blendOption=Linear}
    boneToModify {boneName=stone_09_small}
   }
   n28 = "Transform (Modify) Bone - Bone: stone_01_large"(ComponentPose=@n27, Alpha=1, Translation=Get Shard_Float_D(), Rotation=Get Rotator2(), Scale="1,1,1") {
    rotationMode=BMM_Additive
    translationMode=BMM_Additive
    translationSpace=BCS_BoneSpace
    alphaBoolBlend {blendOption=Linear}
    boneToModify {boneName=stone_01_large}
   }
   n29 = "Transform (Modify) Bone - Bone: stone_02_large"(ComponentPose=@n28, Alpha=1, Translation=Get StoneFloatB(), Rotation=Get Rotator3() -> (Rotator3:Rotator), Scale="1,1,1") {
    rotationMode=BMM_Additive
    translationMode=BMM_Additive
    translationSpace=BCS_BoneSpace
    alphaBoolBlend {blendOption=Linear}
    boneToModify {boneName=stone_02_large}
   }
   n30 = "Transform (Modify) Bone - Bone: stone_03_large"(ComponentPose=@n29, Alpha=1, Translation=Get StoneFloatA(), Rotation=Get Rotator2(), Scale="1,1,1") {
    rotationMode=BMM_Additive
    translationMode=BMM_Additive
    translationSpace=BCS_BoneSpace
    alphaBoolBlend {blendOption=Linear}
    boneToModify {boneName=stone_03_large}
   }
   n31 = "Transform (Modify) Bone - Bone: stone_05_large"(ComponentPose=@n30, Alpha=1, Translation=Get StoneFloatC(), Rotation=Get Rotator2(), Scale="1,1,1") {
    rotationMode=BMM_Additive
    translationMode=BMM_Additive
    translationSpace=BCS_BoneSpace
    alphaBoolBlend {blendOption=Linear}
    boneToModify {boneName=stone_05_large}
   }
   n32 = "Transform (Modify) Bone - Bone: dyn_handle"(ComponentPose=@n31, Alpha=1, Translation=Get StoneFloatA(), Rotation="0,0,0", Scale="1,1,1") {
    rotationMode=BMM_Additive
    rotationSpace=BCS_BoneSpace
    translationMode=BMM_Additive
    translationSpace=BCS_BoneSpace
    alphaBoolBlend {blendOption=Linear}
    boneToModify {boneName=dyn_handle}
   }
   n33 = "Transform (Modify) Bone - Bone: dyn_mouth_piece"(ComponentPose=@n32, Alpha=1, Translation="0,0,0", Rotation=Get Mic Piece Rotator() -> (Mic Piece Rotator:Rotator), Scale="1,1,1") {
    rotationMode=BMM_Additive
    rotationSpace=BCS_BoneSpace
    translationMode=BMM_Additive
    translationSpace=BCS_BoneSpace
    alphaBoolBlend {blendOption=Linear}
    boneToModify {boneName=dyn_mouth_piece}
   }
   n34 = Anim Dynamics - Bone: stone_01_large(ComponentPose=@n33, Alpha=0.7, ExternalForce="0,0,0") {
    angularDampingOverride=0.9
    bAngularSpring=true
    bGravityOverrideInSimSpace=true
    bLinearSpring=true
    bOverrideAngularDamping=true
    bOverrideLinearDamping=true
    gravityScale=0
    linearDampingOverride=0.9
    linearSpringConstant=10
    simulationSpace=BoneRelative
    alphaBoolBlend {blendOption=Linear}
    boundBone {boneName=stone_01_large}
    componentLinearAccScale=<0.2 0.2 0.2>
    componentLinearVelScale=<0.9 0.9 0.9>
    gravityOverride {z=-980}
    physicsBodyDefinitions [{
      sphereCollisionRadius=0
      boundBone {boneName=stone_01_large}
      boxExtents=<50 50 50>
      constraintSetup {angularLimitsMax=<0 0 0>, angularLimitsMin=<0 0 0>, angularTarget=<0 0 0>, linearAxesMax=<0.5 0.5 0.5>, linearAxesMin=<-0.5 -0.5 -0.5>}
      localJointOffset=<0 -3 0>
     }]
    relativeSpaceBone {boneName=handle}
   }
   n35 = Anim Dynamics - Bone: stone_02_large(ComponentPose=@n34, Alpha=0.7, ExternalForce="0,0,0") {
    angularDampingOverride=0.9
    bAngularSpring=true
    bGravityOverrideInSimSpace=true
    bLinearSpring=true
    bOverrideAngularDamping=true
    bOverrideLinearDamping=true
    gravityScale=0
    linearDampingOverride=0.9
    linearSpringConstant=10
    simulationSpace=BoneRelative
    alphaBoolBlend {blendOption=Linear}
    boundBone {boneName=stone_02_large}
    componentLinearAccScale=<0.2 0.2 0.2>
    componentLinearVelScale=<0.9 0.9 0.9>
    gravityOverride {z=-980}
    physicsBodyDefinitions [{
      sphereCollisionRadius=0
      boundBone {boneName=stone_02_large}
      boxExtents=<50 50 50>
      constraintSetup {angularLimitsMax=<0 0 0>, angularLimitsMin=<0 0 0>, angularTarget=<0 0 0>, linearAxesMax=<0.5 0.5 0.5>, linearAxesMin=<-0.5 -0.5 -0.5>}
      localJointOffset=<0 -3 0>
     }]
    relativeSpaceBone {boneName=handle}
   }
   n36 = Anim Dynamics - Bone: stone_03_large(ComponentPose=@n35, Alpha=0.7, ExternalForce="0,0,0") {
    angularDampingOverride=0.9
    bAngularSpring=true
    bGravityOverrideInSimSpace=true
    bLinearSpring=true
    bOverrideAngularDamping=true
    bOverrideLinearDamping=true
    gravityScale=0
    linearDampingOverride=0.9
    linearSpringConstant=10
    alphaBoolBlend {blendOption=Linear}
    boundBone {boneName=stone_03_large}
    componentLinearAccScale=<0.2 0.2 0.2>
    componentLinearVelScale=<0.9 0.9 0.9>
    gravityOverride {z=-980}
    physicsBodyDefinitions [{
      sphereCollisionRadius=0
      boundBone {boneName=stone_03_large}
      boxExtents=<50 50 50>
      constraintSetup {angularLimitsMax=<0 0 0>, angularLimitsMin=<0 0 0>, angularTarget=<0 0 0>, linearAxesMax=<0.5 0.5 0.5>, linearAxesMin=<-0.5 -0.5 -0.5>}
      localJointOffset=<0 -3 0>
     }]
    relativeSpaceBone {boneName=handle}
   }
   n37 = Anim Dynamics - Bone: stone_05_large(ComponentPose=@n36, Alpha=0.8, ExternalForce="0,0,0") {
    angularDampingOverride=0.9
    bAngularSpring=true
    bGravityOverrideInSimSpace=true
    bLinearSpring=true
    bOverrideAngularDamping=true
    bOverrideLinearDamping=true
    gravityScale=0
    linearDampingOverride=0.9
    linearSpringConstant=10
    alphaBoolBlend {blendOption=Linear}
    boundBone {boneName=stone_05_large}
    componentLinearAccScale=<0.2 0.2 0.2>
    componentLinearVelScale=<0.9 0.9 0.9>
    gravityOverride {z=-980}
    physicsBodyDefinitions [{
      sphereCollisionRadius=0
      boundBone {boneName=stone_05_large}
      boxExtents=<50 50 50>
      constraintSetup {angularLimitsMax=<0 0 0>, angularLimitsMin=<0 0 0>, angularTarget=<0 0 0>, linearAxesMax=<0.5 0.5 0.5>, linearAxesMin=<-0.5 -0.5 -0.5>}
      localJointOffset=<0 -3 0>
     }]
    relativeSpaceBone {boneName=handle}
   }
   n38 = Anim Dynamics - Bone: stone_01_small(ComponentPose=@n37, Alpha=0.7, ExternalForce="0,0,0") {
    angularDampingOverride=0.9
    bAngularSpring=true
    bGravityOverrideInSimSpace=true
    bLinearSpring=true
    bOverrideAngularDamping=true
    bOverrideLinearDamping=true
    gravityScale=0
    linearDampingOverride=0.9
    linearSpringConstant=10
    simulationSpace=BoneRelative
    alphaBoolBlend {blendOption=Linear}
    boundBone {boneName=stone_01_small}
    componentLinearAccScale=<0.2 0.2 0.2>
    componentLinearVelScale=<0.9 0.9 0.9>
    gravityOverride {z=-980}
    physicsBodyDefinitions [{
      sphereCollisionRadius=0
      boundBone {boneName=stone_01_small}
      boxExtents=<50 50 50>
      constraintSetup {angularLimitsMax=<0 0 0>, angularLimitsMin=<0 0 0>, angularTarget=<0 0 0>, linearAxesMax=<0.75 0.75 0.75>, linearAxesMin=<-0.75 -0.75 -0.75>}
      localJointOffset=<0 -3 0>
     }]
    relativeSpaceBone {boneName=handle}
   }
   n39 = Anim Dynamics - Bone: stone_02_small(ComponentPose=@n38, Alpha=0.7, ExternalForce="0,0,0") {
    angularDampingOverride=0.9
    bAngularSpring=true
    bGravityOverrideInSimSpace=true
    bLinearSpring=true
    bOverrideAngularDamping=true
    bOverrideLinearDamping=true
    gravityScale=0
    linearDampingOverride=0.9
    linearSpringConstant=10
    simulationSpace=BoneRelative
    alphaBoolBlend {blendOption=Linear}
    boundBone {boneName=stone_02_small}
    componentLinearAccScale=<0.2 0.2 0.2>
    componentLinearVelScale=<0.9 0.9 0.9>
    gravityOverride {z=-980}
    physicsBodyDefinitions [{
      sphereCollisionRadius=0
      boundBone {boneName=stone_02_small}
      boxExtents=<50 50 50>
      constraintSetup {angularLimitsMax=<0 0 0>, angularLimitsMin=<0 0 0>, angularTarget=<0 0 0>, linearAxesMax=<0.75 0.75 0.75>, linearAxesMin=<-0.75 -0.75 -0.75>}
      localJointOffset=<0 -3 0>
     }]
    relativeSpaceBone {boneName=handle}
   }
   n40 = Anim Dynamics - Bone: stone_03_small(ComponentPose=@n39, Alpha=0.7, ExternalForce="0,0,0") {
    angularDampingOverride=0.9
    bAngularSpring=true
    bGravityOverrideInSimSpace=true
    bLinearSpring=true
    bOverrideAngularDamping=true
    bOverrideLinearDamping=true
    gravityScale=0
    linearDampingOverride=0.9
    linearSpringConstant=10
    simulationSpace=BoneRelative
    alphaBoolBlend {blendOption=Linear}
    boundBone {boneName=stone_03_small}
    componentLinearAccScale=<0.2 0.2 0.2>
    componentLinearVelScale=<0.9 0.9 0.9>
    gravityOverride {z=-980}
    physicsBodyDefinitions [{
      sphereCollisionRadius=0
      boundBone {boneName=stone_03_small}
      boxExtents=<50 50 50>
      constraintSetup {angularLimitsMax=<0 0 0>, angularLimitsMin=<0 0 0>, angularTarget=<0 0 0>, linearAxesMax=<0.75 0.75 0.75>, linearAxesMin=<-0.75 -0.75 -0.75>}
      localJointOffset=<0 -3 0>
     }]
    relativeSpaceBone {boneName=handle}
   }
   n41 = Anim Dynamics - Bone: stone_04_small(ComponentPose=@n40, Alpha=0.7, ExternalForce="0,0,0") {
    angularDampingOverride=0.9
    bAngularSpring=true
    bGravityOverrideInSimSpace=true
    bLinearSpring=true
    bOverrideAngularDamping=true
    bOverrideLinearDamping=true
    gravityScale=0
    linearDampingOverride=0.9
    linearSpringConstant=10
    simulationSpace=BoneRelative
    alphaBoolBlend {blendOption=Linear}
    boundBone {boneName=stone_04_small}
    componentLinearAccScale=<0.2 0.2 0.2>
    componentLinearVelScale=<0.9 0.9 0.9>
    gravityOverride {z=-980}
    physicsBodyDefinitions [{
      sphereCollisionRadius=0
      boundBone {boneName=stone_04_small}
      boxExtents=<50 50 50>
      constraintSetup {angularLimitsMax=<0 0 0>, angularLimitsMin=<0 0 0>, angularTarget=<0 0 0>, linearAxesMax=<0.75 0.75 0.75>, linearAxesMin=<-0.75 -0.75 -0.75>}
      localJointOffset=<0 -3 0>
     }]
    relativeSpaceBone {boneName=handle}
   }
   n42 = Anim Dynamics - Bone: stone_05_small(ComponentPose=@n41, Alpha=0.7, ExternalForce="0,0,0") {
    angularDampingOverride=0.9
    bAngularSpring=true
    bGravityOverrideInSimSpace=true
    bLinearSpring=true
    bOverrideAngularDamping=true
    bOverrideLinearDamping=true
    gravityScale=0
    linearDampingOverride=0.9
    linearSpringConstant=10
    simulationSpace=BoneRelative
    alphaBoolBlend {blendOption=Linear}
    boundBone {boneName=stone_05_small}
    componentLinearAccScale=<0.2 0.2 0.2>
    componentLinearVelScale=<0.9 0.9 0.9>
    gravityOverride {z=-980}
    physicsBodyDefinitions [{
      sphereCollisionRadius=0
      boundBone {boneName=stone_05_small}
      boxExtents=<50 50 50>
      constraintSetup {angularLimitsMax=<0 0 0>, angularLimitsMin=<0 0 0>, angularTarget=<0 0 0>, linearAxesMax=<0.75 0.75 0.75>, linearAxesMin=<-0.75 -0.75 -0.75>}
      localJointOffset=<0 -3 0>
     }]
    relativeSpaceBone {boneName=handle}
   }
   n43 = Anim Dynamics - Bone: stone_06_small(ComponentPose=@n42, Alpha=0.7, ExternalForce="0,0,0") {
    angularDampingOverride=0.9
    bAngularSpring=true
    bGravityOverrideInSimSpace=true
    bLinearSpring=true
    bOverrideAngularDamping=true
    bOverrideLinearDamping=true
    gravityScale=0
    linearDampingOverride=0.9
    linearSpringConstant=10
    simulationSpace=BoneRelative
    alphaBoolBlend {blendOption=Linear}
    boundBone {boneName=stone_06_small}
    componentLinearAccScale=<0.2 0.2 0.2>
    componentLinearVelScale=<0.9 0.9 0.9>
    gravityOverride {z=-980}
    physicsBodyDefinitions [{
      sphereCollisionRadius=0
      boundBone {boneName=stone_06_small}
      boxExtents=<50 50 50>
      constraintSetup {angularLimitsMax=<0 0 0>, angularLimitsMin=<0 0 0>, angularTarget=<0 0 0>, linearAxesMax=<0.75 0.75 0.75>, linearAxesMin=<-0.75 -0.75 -0.75>}
      localJointOffset=<0 -3 0>
     }]
    relativeSpaceBone {boneName=handle}
   }
   n44 = Anim Dynamics - Bone: stone_07_small(ComponentPose=@n43, Alpha=0.7, ExternalForce="0,0,0") {
    angularDampingOverride=0.9
    bAngularSpring=true
    bGravityOverrideInSimSpace=true
    bLinearSpring=true
    bOverrideAngularDamping=true
    bOverrideLinearDamping=true
    gravityScale=0
    linearDampingOverride=0.9
    linearSpringConstant=10
    simulationSpace=BoneRelative
    alphaBoolBlend {blendOption=Linear}
    boundBone {boneName=stone_07_small}
    componentLinearAccScale=<0.2 0.2 0.2>
    componentLinearVelScale=<0.9 0.9 0.9>
    gravityOverride {z=-980}
    physicsBodyDefinitions [{
      sphereCollisionRadius=0
      boundBone {boneName=stone_07_small}
      boxExtents=<50 50 50>
      constraintSetup {angularLimitsMax=<0 0 0>, angularLimitsMin=<0 0 0>, angularTarget=<0 0 0>, linearAxesMax=<0.75 0.75 0.75>, linearAxesMin=<-0.75 -0.75 -0.75>}
      localJointOffset=<0 -3 0>
     }]
    relativeSpaceBone {boneName=handle}
   }
   n45 = Anim Dynamics - Bone: stone_08_small(ComponentPose=@n44, Alpha=0.7, ExternalForce="0,0,0") {
    angularDampingOverride=0.9
    bAngularSpring=true
    bGravityOverrideInSimSpace=true
    bLinearSpring=true
    bOverrideAngularDamping=true
    bOverrideLinearDamping=true
    gravityScale=0
    linearDampingOverride=0.9
    linearSpringConstant=10
    simulationSpace=BoneRelative
    alphaBoolBlend {blendOption=Linear}
    boundBone {boneName=stone_08_small}
    componentLinearAccScale=<0.2 0.2 0.2>
    componentLinearVelScale=<0.9 0.9 0.9>
    gravityOverride {z=-980}
    physicsBodyDefinitions [{
      sphereCollisionRadius=0
      boundBone {boneName=stone_08_small}
      boxExtents=<50 50 50>
      constraintSetup {angularLimitsMax=<0 0 0>, angularLimitsMin=<0 0 0>, angularTarget=<0 0 0>, linearAxesMax=<0.75 0.75 0.75>, linearAxesMin=<-0.75 -0.75 -0.75>}
      localJointOffset=<0 -3 0>
     }]
    relativeSpaceBone {boneName=handle}
   }
   n46 = Anim Dynamics - Bone: stone_09_small(ComponentPose=@n45, Alpha=0.7, ExternalForce="0,0,0") {
    angularDampingOverride=0.9
    bAngularSpring=true
    bGravityOverrideInSimSpace=true
    bLinearSpring=true
    bOverrideAngularDamping=true
    bOverrideLinearDamping=true
    gravityScale=0
    linearDampingOverride=0.9
    linearSpringConstant=10
    simulationSpace=BoneRelative
    alphaBoolBlend {blendOption=Linear}
    boundBone {boneName=stone_09_small}
    componentLinearAccScale=<0.2 0.2 0.2>
    componentLinearVelScale=<0.9 0.9 0.9>
    gravityOverride {z=-980}
    physicsBodyDefinitions [{
      sphereCollisionRadius=0
      boundBone {boneName=stone_09_small}
      boxExtents=<50 50 50>
      constraintSetup {angularLimitsMax=<0 0 0>, angularLimitsMin=<0 0 0>, angularTarget=<0 0 0>, linearAxesMax=<0.75 0.75 0.75>, linearAxesMin=<-0.75 -0.75 -0.75>}
      localJointOffset=<0 -3 0>
     }]
    relativeSpaceBone {boneName=handle}
   }
   n47 = Anim Dynamics - Bone: dyn_handle(ComponentPose=@n46, Alpha=0.7, ExternalForce="0,0,0") {
    angularDampingOverride=0.9
    bAngularSpring=true
    bGravityOverrideInSimSpace=true
    bLinearSpring=true
    bOverrideAngularDamping=true
    bOverrideLinearDamping=true
    gravityScale=0
    linearDampingOverride=0.9
    linearSpringConstant=10
    simulationSpace=BoneRelative
    alphaBoolBlend {blendOption=Linear}
    boundBone {boneName=dyn_handle}
    componentLinearAccScale=<0.1 0.1 0.1>
    componentLinearVelScale=<2 2 2>
    physicsBodyDefinitions [{
      sphereCollisionRadius=0
      boundBone {boneName=dyn_handle}
      boxExtents=<30 30 30>
      constraintSetup {angularLimitsMax=<0 0 0>, angularLimitsMin=<0 0 0>, angularTarget=<0 0 0>, linearAxesMax=<0.25 0.25 0.25>, linearAxesMin=<-0.25 -0.25 -0.25>}
      localJointOffset=<0 -3 0>
     }]
    relativeSpaceBone {boneName=handle}
    simSpaceSettings {maxAngularAcceleration=5000, maxAngularVelocity=5000}
   }
   n48 = Anim Dynamics - Bone: dyn_mouth_piece(ComponentPose=@n47, Alpha=0.7, ExternalForce="0,0,0") {
    angularDampingOverride=0.9
    bAngularSpring=true
    bGravityOverrideInSimSpace=true
    bLinearSpring=true
    bOverrideAngularDamping=true
    bOverrideLinearDamping=true
    gravityScale=0
    linearDampingOverride=0.9
    linearSpringConstant=10
    simulationSpace=BoneRelative
    alphaBoolBlend {blendOption=Linear}
    boundBone {boneName=dyn_mouth_piece}
    componentLinearAccScale=<0.1 0.1 0.1>
    componentLinearVelScale=<2 2 2>
    physicsBodyDefinitions [{
      sphereCollisionRadius=0
      boundBone {boneName=dyn_mouth_piece}
      boxExtents=<30 30 30>
      constraintSetup {angularLimitsMax=<0 0 0>, angularLimitsMin=<0 0 0>, angularTarget=<0 0 0>, linearAxesMax=<0.25 0.25 0.25>, linearAxesMin=<-0.25 -0.25 -0.25>}
      localJointOffset=<0 -3 0>
     }]
    relativeSpaceBone {boneName=handle}
    simSpaceSettings {maxAngularAcceleration=5000, maxAngularVelocity=5000}
   }
   Save cached pose 'Cached Pose 1'(Pose=Component To Local(ComponentPose=@n48)) {cacheName=Cached Pose 1}
   Use cached pose 'Cached Pose 1'()
  ]}
 EventGraph {nodes [
   Event BlueprintUpdateAnimation()
   > n1 = Sequence!(execute) -> (then_0, then_1, then_2:Exec)
   n2 = GetGameTimeInSeconds()
   n3 = "Sin (Radians)"(A=float * float(A=float * float(A=1, B=1.5), B=@n2))
   n4 = "Sin (Radians)"(A=float * float(A=float * float(A=1, B=1), B=@n2))
   n5 = "Sin (Radians)"(A=float * float(A=float * float(A=1, B=2), B=@n2))
   Set StoneFloatA(
    execute=@n1.then_1
    StoneFloatA:Vector=MakeVector(X=float * float(A=@n3, B=0.5), Y=float * float(A=@n4, B=0.6), Z=float * float(A=@n5, B=0.22))
   ) -> (Output_Get:Vector)
   n6 = "Cos (Radians)"!(A:double=float * float(A=float * float(A=1, B=1.5), B=@n2)) -> (ReturnValue:double)
   n7 = "Cos (Radians)"(A=float * float(A=float * float(A=1, B=1.78), B=@n2))
   n8 = "Cos (Radians)"(A=float * float(A=float * float(A=1, B=1.15), B=@n2))
   > Set StoneFloatB(StoneFloatB:Vector=MakeVector(X=float * float(A=@n6, B=0.6), Y=float * float(A=@n7, B=0.725), Z=float * float(A=@n8, B=0.85))) -> (Output_Get:Vector)
   n9 = "Sin (Radians)"(A=float * float(A=float * float(A=1, B=1.5), B=@n2))
   n10 = "Sin (Radians)"(A=float * float(A=float * float(A=1, B=1), B=@n2))
   n11 = "Sin (Radians)"(A=float * float(A=float * float(A=1, B=2), B=@n2))
   > Set StoneFloatC(StoneFloatC:Vector=MakeVector(X=float * float(A=@n9, B=1.5), Y=float * float(A=@n10, B=0.6), Z=float * float(A=@n11, B=0.25))) -> (Output_Get:Vector)
   n12 = float / float(A=@n2, B=-5) {is=title K2Node_CallFunction}
   n13 = float / float(A=@n2, B=-20)
   > Set Rotator1(Rotator1:Rotator=MakeRotator(Roll=float * float(A=@n12, B=100), Yaw=float * float(A=@n13, B=1000))) -> (Output_Get:Rotator)
   n14 = float / float(A=@n2, B=-500)
   > Set Rotator2(Rotator2:Rotator=MakeRotator(Yaw=float * float(A=@n14, B=20000))) -> (Output_Get:Rotator)
   n15 = float / float(A=@n2, B=-5)
   n16 = float / float(A=@n2, B=-10)
   > Set Rotator3(Rotator3:Rotator=MakeRotator(Pitch=float * float(A=@n15, B=-100), Yaw=float * float(A=@n16, B=-500))) -> (Output_Get:Rotator)
   n17 = float / float(A=@n2, B=-500)
   > Set Mic Piece Rotator(Mic Piece Rotator:Rotator=MakeRotator(Roll=float * float(A=@n17, B=15000))) -> (Output_Get:Rotator)
   n18 = "Sin (Radians)"(A=float * float(A=float * float(A=1, B=1.5), B=@n2))
   n19 = "Sin (Radians)"(A=float * float(A=float * float(A=1, B=0.95), B=@n2))
   n20 = "Sin (Radians)"(A=float * float(A=float * float(A=0.6, B=1.5), B=@n2))
   > Set Shard_Float_D(Shard_Float_D:Vector=MakeVector(X=float * float(A=@n20, B=0.85), Y=float * float(A=@n19, B=0.1), Z=float * float(A=@n18, B=0.7))) -> (Output_Get:Vector)
   > Set NotPlayingMelee(NotPlayingMelee:bool=NOT Boolean(A=Get bIsPlayingMeleeAnim())) -> (Output_Get:bool)
   TryGetPawnOwner()
  ]}
]
variables [
 Base_Rotator {type=Rotator}
 BaseFloater {type=Vector}
 Mic Piece Rotator {type=Rotator}
 NotPlayingMelee {type=boolean}
 Rotator1 {type=Rotator}
 Rotator2 {type=Rotator}
 Rotator3 {type=Rotator}
 Shard_Float_D {type=Vector}
 StoneFloatA {type=Vector}
 StoneFloatB {type=Vector}
 StoneFloatC {type=Vector}
]
